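# Animation Blueprint for the female medium TacticMushroom character body cosmetic, which copies its pose from the attached parent mesh.
# A Control Rig adjusts that pose using skydiving and parachute states, then Anim Dynamics adds spring-driven sway to both pouch bones.
# A TacticMushroom meta-tag check on initialization and customization load sets Vest Off, which enables an attach_backpack bone offset.
# Each update computes a Gravity Override through a deprecated function, though no AnimGraph node reads it.
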AnimBlueprint /BRCosmetics/Characters/Player/Female/Medium/Bodies/F_MED_TacticMushroom/Meshes/F_MED_TacticMushroom_AnimBP
target_skeleton=&/F_MED_TacticMushroom_Skeleton
graphs [
 AnimGraph {nodes [
   n1 = Control Rig!(
    Source=Copy Pose From Mesh() {bUseAttachedParent=true}
    bAlphaBoolEnabled:bool=True
    IsSkyDiving:bool=Get bIsSkyDiving()
    IsParachuteOpened:bool=Get bIsParachuteOpened()
   ) -> (Pose) {
    alphaInputType=Bool
    alphaBoolBlend {blendOption=Linear}
    controlRigAssetReference {blueprintRigClass=&/F_MED_TacticMushroom_CtrlRig.F_MED_TacticMushroom_CtrlRig_C}
   }
   n2 = Anim Dynamics - Bone: dyn_pouches_l(ComponentPose=Local To Component(LocalPose=@n1), Alpha=1, ExternalForce="0,0,0") {
    angularDampingOverride=1
    angularSpringConstant=300
    bAngularSpring=true
    bOverrideAngularDamping=true
    bUseGravityOverride=true
    gravityScale=0.35
    simulationSpace=RootRelative
    alphaBoolBlend {blendOption=Linear}
    boundBone {boneName=dyn_pouches_l}
    componentAppliedLinearAccClamp=<10000 10000 10000>
    componentLinearAccScale=<1 1 1>
    componentLinearVelScale=<1 1 1>
    gravityOverride {z=-980}
    physicsBodyDefinitions [{
      boundBone {boneName=dyn_pouches_l}
      boxExtents=<10 10 10>
      constraintSetup {angularLimitsMax=<10 10 15>, angularLimitsMin=<-10 -10 -15>, angularTarget=<1 0 0>}
      localJointOffset=<0.5 0 0>
     }]
   }
   n3 = Anim Dynamics - Bone: dyn_pouches_r(ComponentPose=@n2, Alpha=1, ExternalForce="0,0,0") {
    angularDampingOverride=1
    angularSpringConstant=300
    bAngularSpring=true
    bOverrideAngularDamping=true
    bUseGravityOverride=true
    gravityScale=0.35
    simulationSpace=RootRelative
    alphaBoolBlend {blendOption=Linear}
    boundBone {boneName=dyn_pouches_r}
    componentAppliedLinearAccClamp=<10000 10000 10000>
    componentLinearAccScale=<1 1 1>
    componentLinearVelScale=<1 1 1>
    gravityOverride {z=-980}
    physicsBodyDefinitions [{
      boundBone {boneName=dyn_pouches_r}
      boxExtents=<10 10 10>
      constraintSetup {angularLimitsMax=<10 10 15>, angularLimitsMin=<-10 -10 -15>, angularTarget=<1 0 0>}
      localJointOffset=<0.5 0 0>
     }]
   }
   n4 = "Transform (Modify) Bone - Bone: attach_backpack"!(ComponentPose=@n3, bAlphaBoolEnabled:bool=Get Vest Off() -> (Vest Off:bool), Translation="0,3.5,0", Rotation="0,0,0", Scale="1,1,1") -> (Pose) {
    alphaInputType=Bool
    rotationSpace=BCS_ParentBoneSpace
    translationMode=BMM_Replace
    translationSpace=BCS_BoneSpace
    alphaBoolBlend {blendOption=Linear}
    boneToModify {boneName=attach_backpack}
    translation {y=3.5}
   }
   Output Pose(Result=Component To Local(ComponentPose=@n4)) {layerGroup=None}
  ]}
 EventGraph {nodes [
   Event BlueprintUpdateAnimation()
   > n1 = Gravity Override State Function DEPRECATED(CharacterPartAnimInstance=Self-Reference(), GameTimeInSeconds=GetGameTimeInSeconds(), GravityOverrideParams=Get Gravity Override Params())
   > Set Gravity Override(Gravity Override=@n1)
   n2 = Event OnCharacterCustomizationLoaded()
   Set Vest Off(
    execute=[@n2.then, @n3.then]
    Vest Off:bool=CharacterItemDefinitionHasAnyMetaTag(Tags=Make Array("[0]"=(TagName="Cosmetics.MetaTags.TacticMushroom")))
   ) -> (Output_Get:bool)
   n3 = Event BlueprintInitializeAnimation()
  ]}
]
variables [
 Gravity Override {type=Vector}
 Gravity Override Params {type=GravityOverrideParamsStruct}
 Vest Off {type=boolean}
]
